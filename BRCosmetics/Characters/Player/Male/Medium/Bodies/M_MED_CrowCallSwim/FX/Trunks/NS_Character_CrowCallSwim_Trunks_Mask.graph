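# Niagara cosmetic effect for the CrowCallSwim trunks mask, adding flames, embers and flame sprites attached to the character's skeletal mesh.
# A persistent flame mesh feeds time-smoothed bone position into material parameters, while two mirrored spawn groups emit curl-noise sprites.
# A shared system timeline drives flare-up and fade through spawn rates, colors and material parameters.
# Aiming down sights scales everything to zero; front-end state adjusts flame-sprite opacity.
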NiagaraSystem /BRCosmetics/Characters/Player/Male/Medium/Bodies/M_MED_CrowCallSwim/FX/Trunks/NS_Character_CrowCallSwim_Trunks_Mask
dependencies {
 usedDataInterfaces [
  &NiagaraDataInterfaceCurve
  &NiagaraDataInterfaceArrayFloat3
  &NiagaraDataInterfaceColorCurve
  &NiagaraDataInterfaceArrayFloat
  &NiagaraDataInterfaceVector2DCurve
  &NiagaraDataInterfaceSkeletalMesh
 ]
 usedDynamicInputs [
  &NormalizeFloat
  &VectorLength
  &MakeFloat4FromVectorAndFloat
  &FNTimeOfDayScalar
  &SetBoolByIntComparison
 ]
 usedModules [
  &self:SystemSpawnScript.SystemScriptSource.SystemScriptGraph.NiagaraNodeAssignment_9.SetVariables_#1
  &self:SystemSpawnScript.SystemScriptSource.SystemScriptGraph.NiagaraNodeAssignment_0.SetVariables_#2
  &self:SystemSpawnScript.SystemScriptSource.SystemScriptGraph.NiagaraNodeAssignment_3.SetVariables_#3
  &SystemState
  &self:SystemSpawnScript.SystemScriptSource.SystemScriptGraph.NiagaraNodeAssignment_5.SetVariables_#4
  &self:SystemSpawnScript.SystemScriptSource.SystemScriptGraph.NiagaraNodeAssignment_1.SetVariables_#5
  &Timeline
  &self:SystemSpawnScript.SystemScriptSource.SystemScriptGraph.NiagaraNodeAssignment_2.SetVariables_#6
 ]
}
emitters {
 Embers {
  emitterUpdateScript [
   &EmitterState(Life Cycle Mode=::NewEnumerator1, Loop Behavior=::NewEnumerator1, Loop Duration Mode=::NewEnumerator1, UseLoopDelay=false)
   &SpawnRate(
    SpawnRate=&FloatFromCurve(
     FloatCurve=DataInterface() {propertyValues {Curve {keys [
         {arriveTangent=-0.6696, interpMode=RCIM_Cubic, leaveTangent=-0.6696, tangentMode=RCTM_User, time=0, value=1}
         {interpMode=RCIM_Cubic, tangentMode=RCTM_User, time=1, value=0}
        ]}}}
     CurveIndex=$System.Timeline.Time
     Scale Curve=100
    )
    Use Spawn Probability=true
   ) {module=SpawnRate001}
  ]
  particleSpawnScript [
   /SetVariables(Particles.PerParticleRandom:NiagaraFloat=&RandomRangeFloat(Minimum=0, Maximum=1)) {isSetParameters=true}
   &FN_SetSpawnGroup(Number of Spawn Groups=2)
   &InitializeParticle(
    Color=<rgba 1 0 0 1>
    Lifetime=&RandomRangeFloat(Minimum=0.2, Maximum=0.6)
    Sprite Size=&Multiply_Vector2D(
     A=&RandomRangeVector2D(Minimum=<0.15 0.2>, Maximum=<0.2 0.4>)
     B=&MakeVector2D(
      X=&FloatFromCurve(
       FloatCurve=DataInterface() {propertyValues {Curve {keys [<at 0 1>,<at 0.8 2 RCIM_Constant>,<at 0.95 3 RCIM_Constant>]}}}
       CurveIndex=$Particles.PerParticleRandom
      )
      Y=&FloatFromCurve(
       FloatCurve=DataInterface() {propertyValues {Curve {keys [<at 0 1>,<at 0.8 2 RCIM_Constant>,<at 0.95 3>]}}}
       CurveIndex=$Particles.PerParticleRandom
      )
     )
    )
    Sprite Size Mode=::NewEnumerator4
    UsePositionOffset=false
    Write Material Random=true
   )
   &SkeletalMeshLocation(
    Bone Sampling Mode=::NewEnumerator4
    "Inherit Velocity (Scale)"=0.2
    MaskSpawnGroup=false
    Orientation Sampling=::NewEnumerator1
    Sampled Position Offset=&SelectVectorFromArray(
     Direct Array Index=&SpawnGroup()
     Vector Selection Array=DataInterface() {propertyValues {FloatData [<0 10 4>, <0 10 -4>]}}
    )
    Skeletal Mesh=$System.Skeletal Mesh
    Velocity Sampling=::NewEnumerator0
    VelocitySpeedLimit=false
   )
   &ShapeLocation(MaskSpawnGroup=false, Override Seed=false, Sphere Radius=1, UseOwnerScale=true)
   &AddVelocity(
    Cone Angle=90
    Cone Axis=&RotateVectorByQuaternion(
     Vector To Rotate=&SelectVectorFromArray(
      Direct Array Index=&SpawnGroup()
      Vector Selection Array=DataInterface() {propertyValues {FloatData [<1 -0.3 3>, <1 -0.3 -3>]}}
     )
     Quaternion=$Particles.SkeletalMeshLocation.SampledRotation
    )
    Module.UseConeFalloff=false
    Module.UseVelDistribution=false
    Override Seed=false
    Rotation Coordinate Space=::Simulation
    Velocity Mode=::NewEnumerator2
    Velocity Speed=&RandomRangeFloat(Minimum=20, Maximum=40)
   )
  ]
  particleUpdateScript [
   &ParticleState()
   &Drag(
    Drag=&FloatFromCurve(
     FloatCurve=DataInterface() {propertyValues {Curve {keys [
         {interpMode=RCIM_Cubic, tangentMode=RCTM_User, time=0, value=0}
         {arriveTangent=2.3593, interpMode=RCIM_Cubic, leaveTangent=2.3593, tangentMode=RCTM_User, time=1, value=1}
        ]}}}
     Scale Curve=&RandomRangeFloat(Minimum=15, Maximum=25)
    )
    Use Linear Drag=true
    Use Rotational Drag=false
   )
   &CurlNoiseForce(
    Noise Frequency=75
    Noise Strength=&FloatFromCurve(
     FloatCurve=DataInterface() {propertyValues {Curve {keys [{interpMode=RCIM_Cubic, tangentMode=RCTM_User, time=0, value=0}, {interpMode=RCIM_Cubic, tangentMode=RCTM_User, time=1, value=1}]}}}
     Scale Curve=500
    )
    Pan Noise=true
    Pan Noise Field=<0 0 -0.5>
   ) {module=CurlNoiseForce001}
   &GravityForce(Gravity=&MakeVector(
     X=0
     Y=0
     Z=&FloatFromCurve(
      FloatCurve=DataInterface() {propertyValues {Curve {keys [{interpMode=RCIM_Cubic, tangentMode=RCTM_User, time=0, value=0}, {interpMode=RCIM_Cubic, tangentMode=RCTM_User, time=1, value=1}]}}}
      Scale Curve=50
     )
    ))
   &SolveForcesAndVelocity(Clamp Velocity=false, Limit Acceleration=false)
   &ScaleColor(
    Scale RGB=&MakeVectorFromLinearColorRGB(Color=&ColorFromCurve(ColorCurve=DataInterface() {propertyValues {
        AlphaCurve {keys [<at 0.571 1 RCIM_Cubic>,<at 0.9966 0>]}
        BlueCurve {keys [<at 0.0039 3>,<at 1 1 RCIM_Cubic>]}
        GreenCurve {keys [<at 0.0039 3>,<at 1 1 RCIM_Cubic>]}
        RedCurve {keys [<at 0.0039 3>,<at 1 1 RCIM_Cubic>]}
       }}))
    ScaleA=true
    ScaleRGB=true
   )
   &ScaleSpriteSize(
    Uniform Curve Scale=&MakeFloatFromVector(Vector=$Engine.Owner.Scale, Channel=::NewEnumerator0)
    Uniform Curve Sprite Scale=DataInterface() {propertyValues {Curve {keys [
        {tangentMode=RCTM_User, time=0, value=1}
        {arriveTangent=-0.0405, interpMode=RCIM_Cubic, leaveTangent=-0.0405, tangentMode=RCTM_User, time=0.7009, value=1}
        {arriveTangent=-8.0089, interpMode=RCIM_Cubic, leaveTangent=-8.0089, tangentMode=RCTM_User, time=1, value=0}
       ]}}}
   )
   &ScaleSpriteSizeBySpeed(
    Max Scale Factor=<1.666 10>
    Min Scale Factor=<1.333 1.333>
    Sample Scale Factor By Curve=true
    Scale Factor Curve=DataInterface() {propertyValues {Curve {keys [<at 0 0>, {arriveTangent=1.4989, interpMode=RCIM_Cubic, tangentMode=RCTM_Break, time=1, value=1}]}}}
    Velocity Threshold=150
   )
   &ScaleSpriteSize(
    Scale Sprite Size Mode=::NewEnumerator0
    Uniform Scale Factor=&MakeCustomFloatFromBool(
     Bool=$User.bisADS
     True Float=&SelectFloatFromArray(Direct Array Index=$Particles.SpawnGroup, Float Selection Array=DataInterface() {propertyValues {FloatData [0,0]}})
     False Float=1
    )
   ) {module=ScaleSpriteSize001}
  ]
  rendererClasses [&NiagaraSpriteRendererProperties]
  renderers [{rendererClass=&NiagaraSpriteRendererProperties}]
 }
 FlameSprites {
  emitterUpdateScript [
   &EmitterState(Life Cycle Mode=::NewEnumerator1, Loop Behavior=::NewEnumerator1, Loop Delay=1.1, Loop Duration Mode=::NewEnumerator1, UseLoopDelay=false)
   &SpawnRate(
    SpawnRate=&FloatFromCurve(
     FloatCurve=DataInterface() {propertyValues {Curve {keys [
         {arriveTangent=-0.5725, interpMode=RCIM_Cubic, leaveTangent=-0.5725, tangentMode=RCTM_User, time=0, value=1}
         {interpMode=RCIM_Cubic, tangentMode=RCTM_User, time=1, value=0}
        ]}}}
     CurveIndex=$System.Timeline.Time
     Scale Curve=50
    )
    Use Spawn Probability=true
   ) {module=SpawnRate001}
  ]
  particleSpawnScript [
   &InitializeParticle(
    Color Maximum=<rgba 0.03 0 0 0.6>
    Color Minimum=<rgba 0.5 0.025 0 0.4>
    Color Mode=::NewEnumerator2
    Lifetime=&Multiply_Float(
     A=&RandomRangeFloat(Minimum=1, Maximum=1.2)
     B=&FloatFromCurve(
      FloatCurve=DataInterface() {propertyValues {Curve {keys [
          {arriveTangent=1.9127, interpMode=RCIM_Cubic, leaveTangent=1.9127, tangentMode=RCTM_User, time=0, value=1}
          <at 0.2 1.2 RCIM_Cubic>
          <at 1 1 RCIM_Cubic>
         ]}}}
      CurveIndex=$System.Timeline.Time
     )
    )
    Sprite Rotation Mode=::NewEnumerator2
    Sprite Size Mode=::NewEnumerator1
    Sprite UV Mode=::NewEnumerator3
    Uniform Sprite Size Max=7
    UsePositionOffset=false
    Write Material Random=true
   ) {module=InitializeParticle001}
   &FN_SetSpawnGroup(Number of Spawn Groups=2)
   &SkeletalMeshLocation(
    Bone Sampling Mode=::NewEnumerator4
    MaskSpawnGroup=false
    Orientation Sampling=::NewEnumerator1
    Sampled Position Offset=&SelectVectorFromArray(
     Direct Array Index=&SpawnGroup()
     Vector Selection Array=DataInterface() {propertyValues {FloatData [<-1 10 2>, <-1 10 -2>]}}
    )
    Skeletal Mesh=$System.Skeletal Mesh
   )
   &ShapeLocation(MaskSpawnGroup=false, Override Seed=false, Sphere Radius=1, UseOwnerScale=true)
   &CameraOffset(Camera Offset Amount=&RandomRangeFloat(Minimum=0, Maximum=10))
   &AddVelocity(
    Cone Angle=10
    Cone Axis=&RotateVectorByQuaternion(
     Vector To Rotate=&SelectVectorFromArray(
      Direct Array Index=&SpawnGroup()
      Vector Selection Array=DataInterface() {propertyValues {FloatData [<1 -1 2>, <1 -1 -2>]}}
     )
     Quaternion=$Particles.SkeletalMeshLocation.SampledRotation
    )
    Module.UseConeFalloff=true
    Module.UseVelDistribution=false
    Override Seed=false
    Rotation Coordinate Space=::Simulation
    Velocity Mode=::NewEnumerator2
    Velocity Speed=&RandomRangeFloat(Minimum=5, Maximum=25)
   )
  ]
  particleUpdateScript [
   /SetVariables(Particles.SubImageIndex=&FloatFromCurve(FloatCurve=DataInterface() {propertyValues {Curve {keys [<at 0 0>,<at 1 15>]}}})) {isSetParameters=true}
   &ParticleState()
   &ScaleSpriteSize(Scale Factor=&Vector2DFromCurve(
     Vector2Curve=DataInterface() {propertyValues {
       XCurve {keys [
         {arriveTangent=3.6499, interpMode=RCIM_Cubic, leaveTangent=3.6499, tangentMode=RCTM_User, time=0, value=0.4}
         {interpMode=RCIM_Cubic, tangentMode=RCTM_User, time=0.5, value=1}
         {arriveTangent=-5.3495, tangentMode=RCTM_User, time=1, value=0.05}
        ]}
       YCurve {keys [
         {arriveTangent=3.6499, interpMode=RCIM_Cubic, leaveTangent=3.6499, tangentMode=RCTM_User, time=0, value=0.4}
         {interpMode=RCIM_Cubic, tangentMode=RCTM_User, time=0.5, value=1}
         {arriveTangent=-5.3495, tangentMode=RCTM_User, time=1, value=0.05}
        ]}
      }}
     ScaleCurve=<1 1>
    ))
   &ScaleColor(
    Scale Alpha=&Multiply_Float(
     A=&FloatFromCurve(
      FloatCurve=DataInterface() {propertyValues {Curve {keys [
          {arriveTangent=9.0725, interpMode=RCIM_Cubic, leaveTangent=9.0725, tangentMode=RCTM_User, time=0, value=0}
          {arriveTangent=0.0164, interpMode=RCIM_Cubic, leaveTangent=0.0164, tangentMode=RCTM_User, time=0.1, value=1}
          {arriveTangent=-2.2898, interpMode=RCIM_Cubic, leaveTangent=-2.2898, tangentMode=RCTM_User, time=1, value=0}
         ]}}}
      Scale Curve=&FN_FrontEndFloatModifier(Gameplay=0.5, Front End=0.25)
     )
     B=&RandomRangeFloat(Minimum=1, Maximum=1.5)
    )
    Scale RGB=&VectorFromFloat(Value=&FloatFromCurve(FloatCurve=DataInterface() {propertyValues {Curve {keys [
          {arriveTangent=-4.2515, interpMode=RCIM_Cubic, leaveTangent=-4.2515, tangentMode=RCTM_User, time=0, value=1}
          {arriveTangent=-0.0056, interpMode=RCIM_Cubic, leaveTangent=-0.0056, tangentMode=RCTM_User, time=0.5, value=0}
         ]}}}))
    ScaleA=true
    ScaleRGB=true
   ) {module=ScaleColor001}
   &Drag(
    Drag=&FloatFromCurve(
     FloatCurve=DataInterface() {propertyValues {Curve {keys [
         {arriveTangent=4.9807, interpMode=RCIM_Cubic, leaveTangent=4.9807, tangentMode=RCTM_User, time=0, value=0}
         {interpMode=RCIM_Cubic, tangentMode=RCTM_User, time=1, value=2}
        ]}}}
     Scale Curve=6
    )
    Use Linear Drag=true
    Use Rotational Drag=true
   )
   &CurlNoiseForce(
    Noise Frequency=100
    Noise Strength=&FloatFromCurve(
     FloatCurve=DataInterface() {propertyValues {Curve {keys [{interpMode=RCIM_Cubic, tangentMode=RCTM_User, time=0, value=0}, {interpMode=RCIM_Cubic, tangentMode=RCTM_User, time=1, value=1}]}}}
     Scale Curve=300
    )
    Pan Noise=true
    Pan Noise Field=<0 0 0.1>
   )
   &GravityForce(Gravity=&MakeVector(
     X=0
     Y=0
     Z=&FloatFromCurve(
      FloatCurve=DataInterface() {propertyValues {Curve {keys [
          {arriveTangent=2.1832, interpMode=RCIM_Cubic, leaveTangent=2.1832, tangentMode=RCTM_User, time=0, value=0}
          {interpMode=RCIM_Cubic, tangentMode=RCTM_User, time=1, value=1}
         ]}}}
      Scale Curve=25
     )
    ))
   &InheritSourceMovement(
    Applied Position Scale=&VectorFromFloat(Value=&FloatFromCurve(FloatCurve=DataInterface() {propertyValues {Curve {keys [
          {interpMode=RCIM_Cubic, tangentMode=RCTM_User, time=0, value=1}
          {arriveTangent=-2.5265, interpMode=RCIM_Cubic, leaveTangent=-2.5265, tangentMode=RCTM_User, time=1, value=0}
         ]}}}))
    ApplyForce=false
    ApplyPosition=true
    ApplyPositionFalloff=false
   )
   &SolveForcesAndVelocity(Clamp Velocity=false, Limit Acceleration=false)
   &ScaleSpriteSize(
    Scale Sprite Size Mode=::NewEnumerator0
    Uniform Scale Factor=&MakeCustomFloatFromBool(
     Bool=$User.bisADS
     True Float=&SelectFloatFromArray(Direct Array Index=$Particles.SpawnGroup, Float Selection Array=DataInterface() {propertyValues {FloatData [0,0]}})
     False Float=1
    )
   ) {module=ScaleSpriteSize001}
  ]
  rendererClasses [&NiagaraSpriteRendererProperties]
  renderers [{rendererClass=&NiagaraSpriteRendererProperties}]
 }
 Flames_Spline_Mesh001 {
  emitterUpdateScript [
   &EmitterState(
    Inactive Response=::NewEnumerator1
    Life Cycle Mode=::NewEnumerator1
    Loop Behavior=::NewEnumerator1
    Loop Duration Mode=::NewEnumerator1
    UseLoopDelay=false
   )
   &SpawnBurst_Instantaneous(Use Spawn Probability=false, UseLoopCountLimit=false)
  ]
  particleSpawnScript [
   &FN_SetSpawnGroup(Number of Spawn Groups=1)
   &InitializeParticle(
    Color=$System.TOD Flame Color
    Lifetime=&RandomRangeFloat(Minimum=0.45, Maximum=0.6)
    Mesh Scale Mode=::NewEnumerator3
    UsePositionOffset=false
    Write Material Random=true
   )
   &CameraOffset(Camera Offset Amount=1)
   /SetVariables() {isSetParameters=true}
  ]
  particleUpdateScript [
   &ParticleState(Kill Particles When Lifetime Has Elapsed=false)
   &SkeletalMeshLocation(
    Bone / Socket Index=&SpawnGroup()
    Bone Sampling Mode=::NewEnumerator4
    MaskSpawnGroup=false
    Sampled Position Offset=<0.5 -0.5 0>
    Skeletal Mesh=$System.Skeletal Mesh
   ) {module=SkeletalMeshLocation001}
   &InterpolateOverTime(Rate Of Change=7.5, Target Vector Value:Vector3f=$Particles.SkeletalMeshLocation001.SampledPosition, Variable Type=::NewEnumerator2)
   /SetVariables(Particles.Y Offset:NiagaraFloat=&Add_Float(
     A=$Particles.Y Offset
     B=&Lerp_Float(
      A=0
      B=0.02
      Alpha=&FloatFromCurve(
       FloatCurve=DataInterface() {propertyValues {Curve {keys [
           {arriveTangent=2.1926, interpMode=RCIM_Cubic, leaveTangent=2.1926, tangentMode=RCTM_User, time=0, value=0}
           {arriveTangent=0.0606, interpMode=RCIM_Cubic, leaveTangent=0.0606, tangentMode=RCTM_User, time=0.3, value=1}
           {arriveTangent=-0.026, interpMode=RCIM_Cubic, leaveTangent=-0.026, tangentMode=RCTM_User, time=0.7, value=1}
           {arriveTangent=-2.0344, interpMode=RCIM_Cubic, leaveTangent=-2.0344, tangentMode=RCTM_User, time=1, value=0}
          ]}}}
       CurveIndex=$System.Timeline.Time
      )
     )
    )) {isSetParameters=true}
   &DynamicMaterialParameters(
    Index 0 Param 1=&Lerp_Float(
     A=1
     B=2
     Alpha=&FloatFromCurve(
      FloatCurve=DataInterface() {propertyValues {Curve {keys [
          {arriveTangent=10.3645, interpMode=RCIM_Cubic, leaveTangent=10.3645, tangentMode=RCTM_User, time=0, value=0}
          {arriveTangent=-0.0325, interpMode=RCIM_Cubic, leaveTangent=-0.0325, tangentMode=RCTM_User, time=0.2, value=1}
          {arriveTangent=-0.026, interpMode=RCIM_Cubic, leaveTangent=-0.026, tangentMode=RCTM_User, time=0.8, value=1}
          {arriveTangent=-3.2787, interpMode=RCIM_Cubic, leaveTangent=-3.2787, tangentMode=RCTM_User, time=1, value=0}
         ]}}}
      CurveIndex=$System.Timeline.Time
     )
    )
    Index 0 Param 2=&Lerp_Float(
     A=1
     B=0.5
     Alpha=&FloatFromCurve(
      FloatCurve=DataInterface() {propertyValues {Curve {keys [
          {arriveTangent=10.3645, interpMode=RCIM_Cubic, leaveTangent=10.3645, tangentMode=RCTM_User, time=0, value=0}
          {arriveTangent=-0.0325, interpMode=RCIM_Cubic, leaveTangent=-0.0325, tangentMode=RCTM_User, time=0.2, value=1}
          {arriveTangent=-0.026, interpMode=RCIM_Cubic, leaveTangent=-0.026, tangentMode=RCTM_User, time=0.8, value=1}
          {arriveTangent=-3.2787, interpMode=RCIM_Cubic, leaveTangent=-3.2787, tangentMode=RCTM_User, time=1, value=0}
         ]}}}
      CurveIndex=$System.Timeline.Time
     )
    )
    Index 0 Param 3=&Lerp_Float(
     A=1
     B=0.7
     Alpha=&FloatFromCurve(
      FloatCurve=DataInterface() {propertyValues {Curve {keys [
          {arriveTangent=10.3645, interpMode=RCIM_Cubic, leaveTangent=10.3645, tangentMode=RCTM_User, time=0, value=0}
          {arriveTangent=-0.0325, interpMode=RCIM_Cubic, leaveTangent=-0.0325, tangentMode=RCTM_User, time=0.2, value=1}
          {arriveTangent=-0.026, interpMode=RCIM_Cubic, leaveTangent=-0.026, tangentMode=RCTM_User, time=0.8, value=1}
          {arriveTangent=-3.2787, interpMode=RCIM_Cubic, leaveTangent=-3.2787, tangentMode=RCTM_User, time=1, value=0}
         ]}}}
      CurveIndex=$System.Timeline.Time
     )
    )
    Index 0 Param 4=&Lerp_Float(
     A=0.75
     B=1
     Alpha=&FloatFromCurve(
      FloatCurve=DataInterface() {propertyValues {Curve {keys [
          {arriveTangent=10.3645, interpMode=RCIM_Cubic, leaveTangent=10.3645, tangentMode=RCTM_User, time=0, value=0}
          {arriveTangent=-0.0325, interpMode=RCIM_Cubic, leaveTangent=-0.0325, tangentMode=RCTM_User, time=0.2, value=1}
          {arriveTangent=-0.026, interpMode=RCIM_Cubic, leaveTangent=-0.026, tangentMode=RCTM_User, time=0.8, value=1}
          {arriveTangent=-3.2787, interpMode=RCIM_Cubic, leaveTangent=-3.2787, tangentMode=RCTM_User, time=1, value=0}
         ]}}}
      CurveIndex=$System.Timeline.Time
     )
    )
    Index 1 Param 1=&MakeFloatFromVector(Vector=$StackContext.InterpolateOverTime.InterpolateOverTime_V3_.Moving Average, Channel=::NewEnumerator0)
    Index 1 Param 2=&MakeFloatFromVector(Vector=$StackContext.InterpolateOverTime.InterpolateOverTime_V3_.Moving Average, Channel=::NewEnumerator1)
    Index 1 Param 3=&MakeFloatFromVector(Vector=$StackContext.InterpolateOverTime.InterpolateOverTime_V3_.Moving Average, Channel=::NewEnumerator2)
    Index 2 Param 1=$Particles.Y Offset
    Module.Param0WriteEnabled=true
    Module.Param0WriteEnabled1=true
    Module.Param0WriteEnabled2=true
    Module.Param1WriteEnabled=true
    Module.Param1WriteEnabled1=true
    Module.Param1WriteEnabled2=false
    Module.Param2WriteEnabled=true
    Module.Param2WriteEnabled1=true
    Module.Param2WriteEnabled2=false
    Module.Param3WriteEnabled=true
    Module.Param3WriteEnabled1=true
    Module.Param3WriteEnabled2=false
    Write Parameter Index 1=true
    Write Parameter Index 2=true
   ) {module=DynamicMaterialParameters001}
   /SetVariables(Particles.Dynamic Parameter Dark:Vector4f=&MakeVector4(
     X=1
     Y=1
     Z=&Lerp_Float(
      A=1
      B=0.75
      Alpha=&FloatFromCurve(
       FloatCurve=DataInterface() {propertyValues {Curve {keys [
           {arriveTangent=10.3645, interpMode=RCIM_Cubic, leaveTangent=10.3645, tangentMode=RCTM_User, time=0, value=0}
           {arriveTangent=-0.0325, interpMode=RCIM_Cubic, leaveTangent=-0.0325, tangentMode=RCTM_User, time=0.2, value=1}
           {arriveTangent=-0.026, interpMode=RCIM_Cubic, leaveTangent=-0.026, tangentMode=RCTM_User, time=0.8, value=1}
           {arriveTangent=-3.2787, interpMode=RCIM_Cubic, leaveTangent=-3.2787, tangentMode=RCTM_User, time=1, value=0}
          ]}}}
       CurveIndex=$System.Timeline.Time
      )
     )
     W=&Lerp_Float(
      A=0.6
      B=1
      Alpha=&FloatFromCurve(
       FloatCurve=DataInterface() {propertyValues {Curve {keys [
           {arriveTangent=10.3645, interpMode=RCIM_Cubic, leaveTangent=10.3645, tangentMode=RCTM_User, time=0, value=0}
           {arriveTangent=-0.0325, interpMode=RCIM_Cubic, leaveTangent=-0.0325, tangentMode=RCTM_User, time=0.2, value=1}
           {arriveTangent=-0.026, interpMode=RCIM_Cubic, leaveTangent=-0.026, tangentMode=RCTM_User, time=0.8, value=1}
           {arriveTangent=-3.2787, interpMode=RCIM_Cubic, leaveTangent=-3.2787, tangentMode=RCTM_User, time=1, value=0}
          ]}}}
       CurveIndex=$System.Timeline.Time
      )
     )
    )) {isSetParameters=true}
   /SetVariables(Particles.Inner Color:LinearColor=&MultiplyLinearColor(
     Linear Color=<rgba 0.75 0.0187 0.0187 1>
     Scale Factor=&MakeVector4_FromFloat(Float=&FloatFromCurve(
       FloatCurve=DataInterface() {propertyValues {Curve {keys [
           {arriveTangent=-0.0098, interpMode=RCIM_Cubic, leaveTangent=-0.0098, tangentMode=RCTM_User, time=0, value=1}
           <at 0.2 2 RCIM_Cubic>
           {arriveTangent=-2.2328, interpMode=RCIM_Cubic, leaveTangent=-2.2328, tangentMode=RCTM_User, time=1, value=1}
          ]}}}
       CurveIndex=$System.Timeline.Time
      ))
    )) {isSetParameters=true}
   /SetVariables(Particles.Color=&LerpLinearColors(
     StartColor=$System.TOD Flame Color
     EndColor=<rgba 0.02 0.0061 0.0064 1>
     LerpFactor=&FloatFromCurve(
      FloatCurve=DataInterface() {propertyValues {Curve {keys [
          {arriveTangent=10.3645, interpMode=RCIM_Cubic, leaveTangent=10.3645, tangentMode=RCTM_User, time=0, value=0}
          <at 0.2 1 RCIM_Cubic>
          {arriveTangent=-1.3747, interpMode=RCIM_Cubic, leaveTangent=-1.3747, tangentMode=RCTM_User, time=1, value=0}
         ]}}}
      CurveIndex=$System.Timeline.Time
     )
    )) {isSetParameters=true}
   &ScaleMeshSize(Scale Factor=&VectorFromFloat(Value=&MakeCustomFloatFromBool(Bool=$User.bisADS, True Float=0, False Float=1))) {module=ScaleMeshSize001}
  ]
  rendererClasses [&NiagaraMeshRendererProperties]
  renderers [
   {rendererClass=&NiagaraMeshRendererProperties, rendererIndex=0}
   {rendererClass=&NiagaraMeshRendererProperties, rendererIndex=1}
  ]
 }
}
summary {
 systemName=NS_Character_CrowCallSwim_Trunks_Mask
 emitters [
  {emitterName=Flames_Spline_Mesh001, rendererClasses [&NiagaraMeshRendererProperties]}
  {emitterName=Embers, rendererClasses [&NiagaraSpriteRendererProperties]}
  {emitterName=FlameSprites, rendererClasses [&NiagaraSpriteRendererProperties]}
 ]
 userVariables [
  User.bIsADS {type=NiagaraBool}
  User.bIsFrontEnd {type=NiagaraBool}
  User.bIsFrontEndPreview {type=NiagaraBool}
  User.bShowEyes {type=NiagaraBool, defaultValue=true}
  User.DebugReactivity {type=NiagaraBool}
  User.KillReactivityCount {type=NiagaraInt32}
 ]
}
